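; Static mesh actor Blueprint for the ClayTile FaceStatue A prop, a variant that uses NoRVT materials (by name, without Runtime Virtual Texture).
; The construction script calls its parent, then overrides material slots 0 and 1 on the static mesh component with the FaceStatue A and B NoRVT material instances.
; BeginPlay, overlap and tick events are disabled placeholders, so it adds no runtime logic.
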(event EventBeginPlay
  (|CallParentFunction) ; This node is disabled and will not be called.\nDrag off pins to build functionality.
)

(event Collision|EventActorBeginOverlap
  (params (OtherActor :type Actor%20Object%20Reference))
  (|CallParentFunction OtherActor) ; This node is disabled and will not be called.\nDrag off pins to build functionality.
)

(event EventTick
  (params (DeltaSeconds :type Float%20%28single-precision%29))
  (|CallParentFunction DeltaSeconds) ; This node is disabled and will not be called.\nDrag off pins to build functionality.
)

(fn ConstructionScript
  (bind _staticmeshcomponent (Variables|StaticMeshActor|GetStaticMeshComponent))
  (|CallParentFunction)
  (Rendering|Material|SetMaterial _staticmeshcomponent 0 "/Hermes/Props/ClayTile/ClayTile_FaceStatue_A/Materials/MI_ClayTile_FaceStatue_A_NoRVT.MI_ClayTile_FaceStatue_A_NoRVT")
  (Rendering|Material|SetMaterial _staticmeshcomponent 1 "/Hermes/Props/ClayTile/ClayTile_FaceStatue_A/Materials/MI_ClayTile_FaceStatue_B_NoRVT.MI_ClayTile_FaceStatue_B_NoRVT"))
